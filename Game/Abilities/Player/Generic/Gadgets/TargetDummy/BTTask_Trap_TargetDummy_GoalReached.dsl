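; Behavior tree task that advances the Target Dummy trap gadget along its path when it reaches a goal.
; It gets the dummy from the AI controller's pawn and increments its current path point index.
; Past the last path location, a looping path resets the index to zero, while a non-looping path force-kills the dummy without DBNO.
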(event EventReceiveExecute
  (params (OwnerActor :type Actor%20Object%20Reference))
  (bind _asaicontroller (Utilities|Casting|CastToAIController OwnerActor))
  (bind _astrap_target_dummy (Utilities|Casting|CastToTrap_TargetDummy (Pawn|GetControlledPawn _asaicontroller)))
  (bind _result (Math|Integer|IncrementInt (Variables|/Game/Abilities/Player/Generic/Gadgets/TargetDummy/Trap_TargetDummy.Trap_TargetDummy_C|GetCurrentPathPointIndex _astrap_target_dummy))
    (:%20%20
      (if (>= _result (Utilities|Array|Length (Variables|/Game/Abilities/Player/Generic/Gadgets/TargetDummy/Trap_TargetDummy.Trap_TargetDummy_C|GetPathLocations _astrap_target_dummy)))
        (if (Variables|/Game/Abilities/Player/Generic/Gadgets/TargetDummy/Trap_TargetDummy.Trap_TargetDummy_C|GetPathLoops _astrap_target_dummy)
          (Variables|/Game/Abilities/Player/Generic/Gadgets/TargetDummy/Trap_TargetDummy.Trap_TargetDummy_C|SetCurrentPathPointIndex 0 _astrap_target_dummy)
          (else
            (AI|ForceKillNoDBNO _astrap_target_dummy)))
        (AI|BehaviorTree|FinishExecute true)
        (else
          _)))))
