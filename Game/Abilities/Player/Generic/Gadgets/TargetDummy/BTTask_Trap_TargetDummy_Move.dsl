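; Behavior tree task that supplies the next movement destination for the Trap_TargetDummy gadget's AI.
; On execute it gets the controlled dummy from the AI controller and checks its StandIdle flag.
; If not idle, it writes the dummy's current path point location to the MovementKey blackboard vector and succeeds; otherwise it fails.
; It only reads the path index and does not move the pawn itself.
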(event EventReceiveExecute
  (params (OwnerActor :type Actor%20Object%20Reference))
  (bind _asaicontroller (Utilities|Casting|CastToAIController OwnerActor))
  (bind _astrap_target_dummy (Utilities|Casting|CastToTrap_TargetDummy (Pawn|GetControlledPawn _asaicontroller)))
  (if (not (Variables|/Game/Abilities/Player/Generic/Gadgets/TargetDummy/Trap_TargetDummy.Trap_TargetDummy_C|GetStandIdle _astrap_target_dummy))
    (AI|BehaviorTree|SetBlackboardValueasVector (Variables|Default|GetMovementKey) (Utilities|Array|Get(aref) (Variables|/Game/Abilities/Player/Generic/Gadgets/TargetDummy/Trap_TargetDummy.Trap_TargetDummy_C|GetPathLocations _astrap_target_dummy) (Variables|/Game/Abilities/Player/Generic/Gadgets/TargetDummy/Trap_TargetDummy.Trap_TargetDummy_C|GetCurrentPathPointIndex _astrap_target_dummy)))
    (AI|BehaviorTree|FinishExecute true)
    (else
      (AI|BehaviorTree|FinishExecute false))))
