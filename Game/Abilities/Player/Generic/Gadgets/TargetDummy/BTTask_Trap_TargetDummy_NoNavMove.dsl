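; Behavior tree task that moves the Trap_TargetDummy gadget straight toward a blackboard vector location without navigation pathfinding.
; On execute it caches the controlled dummy, turns only its yaw to face the target once, then starts a looping timer.
; Each tick of the timer feeds normalized direction into the movement component, or finishes successfully once within 100 units.
; The file never clears the timer and has no abort handler.
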(event EventReceiveExecute
  (params (OwnerActor :type Actor%20Object%20Reference))
  (bind _asaicontroller (Utilities|Casting|CastToAIController OwnerActor))
  (bind _astrap_target_dummy (Utilities|Casting|CastToTrap_TargetDummy (Pawn|GetControlledPawn _asaicontroller)))
  (Variables|Default|SetTargetDummy _astrap_target_dummy)
  (CallFunction|RotatetoTargetLocation)
  (Utilities|Time|SetTimerbyEvent (delegate Custom|MoveToLocation) (Variables|Default|GetMoveUpdateInterval) true))

(event Custom|NearLocation
  (AI|BehaviorTree|FinishExecute true))

(event Custom|MoveToLocation
  (bind _distance (Class|MissionBlueprintFunctionLibrary|DistanceBetweenTwoVectors (Transformation|GetActorLocation (Variables|Default|GetTargetDummy)) (AI|BehaviorTree|GetBlackboardValueasVector (Variables|Default|GetMovementKey))))
  (if (<= _distance 1e+02)
    (CallFunction|NearLocation)
    (else
      (Pawn|Components|PawnMovement|AddInputVector (Variables|/Script/Engine.Character|GetCharacterMovement (Variables|Default|GetTargetDummy)) (Math|Vector|Normalize (- (AI|BehaviorTree|GetBlackboardValueasVector (Variables|Default|GetMovementKey)) (Transformation|GetActorLocation (Variables|Default|GetTargetDummy)))) false))))

(event Custom|RotateToTargetLocation
  (Transformation|SetActorRotation (Variables|Default|GetTargetDummy) (Math|Rotator|MakeRotator (.roll (Transformation|GetActorRotation (Variables|Default|GetTargetDummy))) (.pitch (Transformation|GetActorRotation (Variables|Default|GetTargetDummy))) (.yaw (Math|Rotator|FindLookatRotation (Transformation|GetActorLocation (Variables|Default|GetTargetDummy)) (AI|BehaviorTree|GetBlackboardValueasVector (Variables|Default|GetMovementKey))))) false))
