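# Behavior tree for the Target Dummy trap gadget in the player's abilities, driving the dummy to a goal location.
# A single root Sequence runs a Move task, a NoNavMove task and a GoalReached task in order, using the trap's own blackboard.
# Both movement tasks read the MoveLocation key, with NoNavMove apparently covering movement without navigation at a short update interval.
# The tree has no decorators or services, so all logic lives in the custom tasks.
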BehaviorTree /Game/Abilities/Player/Generic/Gadgets/TargetDummy/BT_Trap_TargetDummy
blackboard=&/BB_Trap_TargetDummy
root {
 name=Sequence
 class=BTComposite_Sequence
 children [
  BTTask_Trap_TargetDummy_Move_C_0 {
   class=BTTask_Trap_TargetDummy_Move_C
   children []
   decorators []
   properties {
    bIgnoreRestartSelf=false
    bShowPropertyDetails=true
    customDescription=""
    nodeName=""
    movementKey {bNoneIsAllowedValue=false, selectedKeyName=MoveLocation}
    tickInterval {interval=-1}
   }
   services []
  }
  BTTask_Trap_TargetDummy_NoNavMove_C_0 {
   class=BTTask_Trap_TargetDummy_NoNavMove_C
   children []
   decorators []
   properties {
    bIgnoreRestartSelf=false
    bShowPropertyDetails=true
    customDescription=""
    moveUpdateInterval=0.01
    nodeName=""
    targetDummy=null
    movementKey {bNoneIsAllowedValue=false, selectedKeyName=MoveLocation}
    tickInterval {interval=-1}
   }
   services []
  }
  BTTask_Trap_TargetDummy_GoalReached_C_0 {
   class=BTTask_Trap_TargetDummy_GoalReached_C
   children []
   decorators []
   properties {bIgnoreRestartSelf=false, bShowPropertyDetails=true, customDescription="", nodeName="", tickInterval {interval=-1}}
   services []
  }
 ]
 decorators []
 properties {children [
   {childTask=&self:BTTask_Trap_TargetDummy_Move_C_0}
   {childTask=&self:BTTask_Trap_TargetDummy_NoNavMove_C_0}
   {childTask=&self:BTTask_Trap_TargetDummy_GoalReached_C_0}
  ]}
}
