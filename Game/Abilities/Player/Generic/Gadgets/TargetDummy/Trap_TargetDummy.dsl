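; Target dummy gadget character that hides its health bar and plays teleport effects and a beam-out sound.
; BeginPlay calls the parent, then spawns the Telenade emitter and sound at the actor's location only without authority.
; SpawnVFX, noted as a temporary death effect, spawns two teleport emitters at the FX_Feet socket, but leaves the sound location at default.
; ShouldUpdateSkeletalMesh always returns true.
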(event EventBeginPlay
  (bind _returnvalue (Transformation|GetActorLocation))
  (|CallParentFunction)
  (Variables|HUD|SetbHideHealthBar true)
  (Networking|SwitchHasAuthority
    (:Remote
      (Effects|SpawnEmitteratLocation "/Game/Athena/Items/Consumables/Telenade/P_Telenade_teleport_temp.P_Telenade_teleport_temp" _returnvalue)
      (Audio|PlaySoundatLocation "/Game/Sounds/Fort_Gadgets/Player_BeamOut/Player_BeamOut_Cue.Player_BeamOut_Cue" _returnvalue))))

(event Custom|SpawnVFX
  (bind _mesh (Variables|Character|GetMesh))
  (bind _returnvalue (Transformation|GetSocketLocation _mesh "FX_Feet"))
  ;[1 1 1] Temporary method to spawn death VFX
  (Effects|SpawnEmitteratLocation "/Game/Effects/Fort_Effects/Effects/Characters/Generic/ResIn/P_Player_Teleport_In_Athena.P_Player_Teleport_In_Athena" _returnvalue)
  (Effects|SpawnEmitteratLocation "/Game/Athena/Items/Consumables/Telenade/P_Telenade_teleport_temp.P_Telenade_teleport_temp" _returnvalue)
  (Audio|PlaySoundatLocation "/Game/Sounds/Fort_Gadgets/Player_BeamOut/Player_BeamOut_Cue.Player_BeamOut_Cue"))

(fn ShouldUpdateSkeletalMesh
  (return true))

(fn ConstructionScript
  (|CallParentFunction))
